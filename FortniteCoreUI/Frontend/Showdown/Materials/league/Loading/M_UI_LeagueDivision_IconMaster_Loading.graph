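# Additive UI material for a league division icon loading visual, built from a swappable icon texture and a border mask.
# The icon is scaled about its center, panned vertically by Speed and thresholded with SmoothStep for an adjustable soft edge.
# A distance-field button-border texture, panned at Speed divided by MaskScale, is thresholded and multiplies the icon.
# ScreenAlignedUVs is unused, and no connections to material outputs are shown, including for Color.
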Material /FortniteCoreUI/Frontend/Showdown/Materials/league/Loading/M_UI_LeagueDivision_IconMaster_Loading
graph {
 properties {bUsedWithStaticMesh=false, blendMode=BLEND_Additive, materialDomain=MD_UI, pixelDepthOffsetMode=PDOM_Legacy, usedShadingModels=Default Lit}
 nodes [
  MaterialFunctionCall(Texture Dimensions=Constant2Vector() {g=42, r=42}) {materialFunction=&ScreenAlignedUVs}
  n1 = ScalarParameter() {defaultValue=0.8, parameterName=IconScale}
  n2 = ScalarParameter() {parameterName=Speed}
  n3 = Panner(
   Coordinate=MaterialFunctionCall(Texture Scale=@n1) {materialFunction=&ScaleUVsByCenter}.UVs
   Speed=AppendVector(A=Constant(), B=@n2)
  ) {speedY=0.25}
  n4 = ScalarParameter() {defaultValue=0.1627, parameterName=Edge}
  n5 = ScalarParameter() {defaultValue=0.1, parameterName=EdgeSoftness}
  n6 = ScalarParameter() {defaultValue=2, parameterName=MaskScale}
  n7 = MaterialFunctionCall(Texture Scale=Multiply(A=@n1, B=@n6)) {materialFunction=&ScaleUVsByCenter}
  n8 = AppendVector(A=Constant(), B=Divide(A=@n2, B=@n6))
  n9 = MaterialFunctionCall(
   Alpha=TextureSample(UVs=Panner(Coordinate=@n7.UVs, Speed=@n8) {speedY=0.25}) {texture=/FortUILibrary/Input/Misc/T_Buttonborder_DF}.R
   Min=ScalarParameter() {defaultValue=0.9, parameterName=MaskEdge}
  ) {materialFunction=&SmoothStep}
  Multiply(
   A=MaterialFunctionCall(
    Alpha=TextureSampleParameter2D(UVs=@n3) {parameterName=Texture, texture=/FortniteCoreUI/Frontend/Showdown/Materials/league/T_UI_LeagueDivision_BasicIcon1}.RGB
    Min=@n4
    Max=Add(A=@n4, B=@n5)
   ) {materialFunction=&SmoothStep}.Result
   B=@n9.Result
  )
  OneMinus(None=@n9.Result)
  VectorParameter() {parameterName=Color, defaultValue=<rgba 1 1 1 1>}
 ]
}
